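# Niagara swing trail for the SlimyTune pickaxe, drawing green ribbons from skeletal mesh bones or sockets through AnimTrail modules.
# NE_BodyTrail draws a faint, wider ribbon, while NE_TipTrail layers three renderers with derived secondary and tertiary colors and widths, plus narrowing and camera offset.
# Both emitters adjust color by time of day.
# A scratch module flags system deactivation, triggering a curve-based alpha taper by normalized particle index on both trails.
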NiagaraSystem /BRCosmetics/Weapons/FORT_Melee/Pickaxe_SlimyTune/FX/NS_Pickaxe_SlimyTune_Trail
dependencies {
 usedDataInterfaces [&NiagaraDataInterfaceCurve, &NiagaraDataInterfaceSkeletalMesh]
 usedDynamicInputs []
 usedModules [
  &self:SystemSpawnScript.SystemScriptSource.SystemScriptGraph.NiagaraNodeAssignment_1.SetVariables_#1
  &SystemState
  &self:SystemSpawnScript.SystemScriptSource.SystemScriptGraph.NiagaraNodeAssignment_6.SetVariables_#2
  &self:ScratchModule
 ]
}
emitters {
 NE_BodyTrail {
  emitterUpdateScript [
   &EmitterState()
   &FN_SpawnAnimTrail_v2()
  ]
  particleSpawnScript [
   /Game/Effects/Niagara/Modules/AnimTrail/FN_InitializeAnimTrail/FN_InitializeAnimTrailBoneIndecies(Second Bone or Socket Index=1)
   &FN_InitializeAnimTrail(AnimTrailWidth=1.2, Color=<rgba 0.155 0.3 0 1>, Lifetime=0.22, SkeletalMesh=$System.Skeletal Mesh) {module=FN_InitializeAnimTrail001}
   /SetVariables(Particles.TrailAlphaFade=0.25) {isSetParameters=true}
  ]
  particleUpdateScript [
   /SetVariables(Particles.TrailAlphaFade:NiagaraFloat=&Multiply_Float(
     A=$Particles.TrailAlphaFade
     B=&MakeCustomFloatFromBool(
      Bool=$System.ScratchModule.SystemDeactivatedThisFrame
      True Float=&FloatFromCurve(
       FloatCurve=DataInterface() {propertyValues {Curve {keys [
           {tangentMode=RCTM_User, time=0, value=1}
           {arriveTangent=-0.0405, interpMode=RCIM_Cubic, leaveTangent=-0.0405, tangentMode=RCTM_User, time=0.7009, value=1}
           {arriveTangent=-8.0089, interpMode=RCIM_Cubic, leaveTangent=-8.0089, tangentMode=RCTM_User, time=1, value=0}
          ]}}}
       CurveIndex=&ReturnNormalizedExecIndex()
      )
      False Float=1
     )
    )) {isSetParameters=true}
   &FN_UpdateAnimTrail(SkeletalMesh=$System.Skeletal Mesh)
   &ScaleColor(
    Scale Alpha=&Multiply_Float(
     A=&FloatFromCurve(
      FloatCurve=DataInterface() {propertyValues {Curve {keys [{interpMode=RCIM_Cubic, tangentMode=RCTM_User, time=0, value=1}, {interpMode=RCIM_Cubic, tangentMode=RCTM_User, time=1, value=0}]}}}
      Scale Curve=$Particles.TrailAlphaFade
     )
     B=&FNTimeOfDayScalar(Night Scale=1, Morning Scale=0.75, Evening Scale=0.75)
    )
    Scale RGB=&VectorFromFloat(Value=&FNTimeOfDayScalar(Night Scale=0.5, Morning Scale=1, Evening Scale=1))
    ScaleA=true
    ScaleRGB=true
   )
  ]
  rendererClasses [&NiagaraRibbonRendererProperties]
  renderers [{rendererClass=&NiagaraRibbonRendererProperties}]
 }
 NE_TipTrail {
  emitterUpdateScript [
   &EmitterState()
   &FN_SpawnAnimTrail(AnimTrailWidthMode=::FromFirst, Spawn Interval=0)
  ]
  particleSpawnScript [
   /Game/Effects/Niagara/Modules/AnimTrail/FN_InitializeAnimTrailBoneIndecies_v2/FN_InitializeAnimTrailBoneIndecies()
   &FN_InitializeAnimTrail(AnimTrailWidth=0.2, Color=<rgba 0.6628 1.2 0.24 1>, Lifetime=0.22, SkeletalMesh=$System.Skeletal Mesh)
   /SetVariables(Particles.TrailAlphaFade=1) {isSetParameters=true}
  ]
  particleUpdateScript [
   /SetVariables(Particles.TrailAlphaFade:NiagaraFloat=&Multiply_Float(
     A=$Particles.TrailAlphaFade
     B=&MakeCustomFloatFromBool(
      Bool=$System.ScratchModule.SystemDeactivatedThisFrame
      True Float=&FloatFromCurve(
       FloatCurve=DataInterface() {propertyValues {Curve {keys [
           {arriveTangent=-0.0405, interpMode=RCIM_Cubic, leaveTangent=-0.0405, tangentMode=RCTM_User, time=0, value=1}
           {arriveTangent=-3.155, interpMode=RCIM_Cubic, leaveTangent=-3.155, tangentMode=RCTM_User, time=1, value=0}
          ]}}}
       CurveIndex=&ReturnNormalizedExecIndex()
      )
      False Float=1
     )
    )) {isSetParameters=true}
   &FN_UpdateAnimTrail(SkeletalMesh=$System.Skeletal Mesh)
   &ScaleColor(
    Scale Alpha=&FloatFromCurve(
     FloatCurve=DataInterface() {propertyValues {Curve {keys [{arriveTangent=-1.1111, time=0, value=0}, {arriveTangent=-1.6414, time=0.1, value=1}, <at 1 0>]}}}
     Scale Curve=$Particles.TrailAlphaFade
    )
    Scale RGB=&Multiply_Vector(
     A=&VectorFromFloat(Value=&FloatFromCurve(FloatCurve=DataInterface() {propertyValues {Curve {keys [
           {interpMode=RCIM_Cubic, leaveTangent=-1.0628, tangentMode=RCTM_Break, time=0, value=1}
           {arriveTangent=-0.0532, leaveTangent=-8.0089, tangentMode=RCTM_User, time=1, value=0.5}
          ]}}}))
     B=&Multiply_Vector(
      A=&VectorFromFloat(Value=&FNTimeOfDayScalar(Night Scale=1, Morning Scale=0.7, Day Scale=0.5, Evening Scale=0.2))
      B=&MakeVectorFromLinearColorRGB(Color=&FNTimeOfDayColor(Night Color=<rgba 0.2686 0.3205 0.5729 1>, Morning Color=<rgba 1 1 1 1>, Day Color=<rgba 1 1 1 1>, Evening Color=<rgba 1 1 1 1>))
     )
    )
    ScaleA=true
    ScaleRGB=true
   )
   &ParticleState()
   /SetVariables(Particles.RibbonU0Override=&Multiply_Float(A=$Engine.Time, B=4)) {isSetParameters=true}
   &CameraOffset(Camera Offset Amount=10)
   &ScaleRibbonWidth(Ribbon Width Scale=&FloatFromCurve(FloatCurve=DataInterface() {propertyValues {Curve {keys [
         {arriveTangent=-2.083, interpMode=RCIM_Cubic, leaveTangent=-2.083, tangentMode=RCTM_User, time=0.25, value=1}
         {tangentMode=RCTM_User, time=1, value=0.25}
        ]}}}))
   /SetVariables(
    Particles.SecondaryColor:LinearColor=&MultiplyLinearColor(
     Linear Color=<rgba 0.2583 0.5 0 0.8>
     Scale Factor=&MakeFloat4FromVectorAndFloat(
      "Vector (RGB)"=<1 1 1>
      "Float (Alpha)"=&Multiply_Float(
       A=&MakeFloatFromLinearColor(LinearColor=$Particles.Color, Channel:ENiagaraLinearColor_Channels=::NewEnumerator3)
       B=0.5
      )
     )
    )
    Particles.SecondaryWidth:NiagaraFloat=&Multiply_Float(A=$Particles.RibbonWidth, B=1.2)
   ) {isSetParameters=true}
   /SetVariables(
    Particles.TertiaryColor:LinearColor=&MultiplyLinearColor(
     Linear Color=<rgba 0.0052 0.01 0 0.8>
     Scale Factor=&MakeFloat4FromVectorAndFloat(
      "Vector (RGB)"=<1 1 1>
      "Float (Alpha)"=&Multiply_Float(
       A=&MakeFloatFromLinearColor(LinearColor=$Particles.Color, Channel:ENiagaraLinearColor_Channels=::NewEnumerator3)
       B=3
      )
     )
    )
    Particles.TertiaryWidth:NiagaraFloat=&Multiply_Float(A=$Particles.RibbonWidth, B=1.3)
   ) {isSetParameters=true}
  ]
  rendererClasses [&NiagaraRibbonRendererProperties]
  renderers [
   {rendererClass=&NiagaraRibbonRendererProperties}
   {rendererClass=&NiagaraRibbonRendererProperties, rendererIndex=1}
   {rendererClass=&NiagaraRibbonRendererProperties, rendererIndex=2}
  ]
 }
}
summary {
 systemName=NS_Pickaxe_SlimyTune_Trail
 emitters [
  {emitterName=NE_BodyTrail, rendererClasses [&NiagaraRibbonRendererProperties]}
  {emitterName=NE_TipTrail, rendererClasses [&NiagaraRibbonRendererProperties]}
 ]
 userVariables [
  User.Debug {type=NiagaraBool}
  User.RibbonIDOffset {type=NiagaraInt32}
  User.VariantPicker {type=NiagaraFloat, defaultValue=0, description=""}
 ]
}
