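# Physics asset for the SlimyTune two-handed pickaxe cosmetic, defining collision and joints for a dangling strap on the axe.
# One large capsule covers the handle bone, and three small capsules cover the strap bones.
# Constraints chain the strap segments from the handle outward, sharing wide swing and narrow twist limit angles.
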PhysicsAsset /BRCosmetics/Weapons/FORT_Melee/Pickaxe_SlimyTune/Meshes/SlimyTune_Axe_PhysicsAsset
bodies {
 Melee_TwoHanded_Handle [{shapeType=Capsule, length=94.0482, radius=13.0661, center=<1.0483 2.4911 42.9845>, rotation=<rot 2.3195 -90 -0.3194>}]
 dyn_strap_1 [{
   shapeType=Capsule
   length=2.7512
   radius=2.6647
   shapeName=dyn_strap_1_capsule
   center=<2.2519 0.3777 -0.2121>
   rotation=<rot -0.8691 91.2377 -79.6828>
  }]
 dyn_strap_2 [{
   shapeType=Capsule
   length=2.0728
   radius=3.0139
   shapeName=dyn_strap_2_capsule
   center=<2.9802 -0.0892 -0.2504>
   rotation=<rot -1.4896 88.74 -89.5292>
  }]
 dyn_strap_3 [{
   shapeType=Capsule
   length=1.8625
   radius=2.9812
   shapeName=dyn_strap_3_capsule
   center=<5.0277 0.3712 -0.3462>
   rotation=<rot -1.5419 88.5748 -89.5494>
  }]
}
constraints [
 {bone1Name=dyn_strap_1, bone2Name=Melee_TwoHanded_Handle, swing1LimitDegrees=80, swing2LimitDegrees=80, twistLimitDegrees=15}
 {bone1Name=dyn_strap_2, bone2Name=dyn_strap_1, swing1LimitDegrees=80, swing2LimitDegrees=80, twistLimitDegrees=15}
 {bone1Name=dyn_strap_3, bone2Name=dyn_strap_2, swing1LimitDegrees=80, swing2LimitDegrees=80, twistLimitDegrees=15}
]
